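; Projectile for the Tracker Sniper that leaves a pulse indicator where it stops, to show pulse markers for a limited time.
; On projectile stop, the authority spawns the pulse indicator building actor, yawed 90 degrees, passes it the owner weapon and starts its markers.
; It attaches a spawned sniper FX actor to the indicator and marks it ready.
; A scalable lifespan value sets the indicator's life and the delay before markers stop.
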(event OnProjectileStop(ProjectileMovementComponent)
  (params (ImpactResult :type Hit%20Result%20Structure%20%28by%20ref%29))
  (Networking|SwitchHasAuthority
    (:Authority
      (bind _returnvalue (CallFunction|GetOwnerWeapon))
      (bind _returnvalue_2 (Building|SpawnBuildingGameplayActor "/VividRazorExotics/Gameplay/TrackerSniper/BGA_Athena_Pulse_SniperTracker_Child.BGA_Athena_Pulse_SniperTracker_Child_C" (Math|Transform|MakeTransform (.location (Transformation|GetActorTransform)) (Math|Rotator|CombineRotators (.rotation (Transformation|GetActorTransform)) "0,90.000000, 0")) (Game|GetInstigator _returnvalue)))
      (Actor|SetLifeSpan _returnvalue_2 (ScalableFloat|GetValueatLevel (Variables|Default|GetLifespan)))
      (bind _asbga_pulse_indicator (Utilities|Casting|CastToBGA_PulseIndicator _returnvalue_2))
      ;[1 1 1] Pass Owning Weapon
      (bind _returnvalue_3 (CallFunction|GetOwnerWeapon))
      (Variables|/Game/Athena/Items/Consumables/PulseIndicator/BGA_PulseIndicator.BGA_PulseIndicator_C|SetOwnerWeapon _returnvalue_3 _asbga_pulse_indicator)
      (Class|BGAPulseIndicator|StartPulseMarkers _asbga_pulse_indicator)
      (bind _returnvalue_4 (Game|SpawnActorfromClass "/VividRazorExotics/Gameplay/TrackerSniper/BP_MoonFlax_Sniper_FX.BP_MoonFlax_Sniper_FX_C" (Math|Transform|MakeTransform (.location (Transformation|GetActorTransform)) (.rotation (Transformation|GetActorTransform))) "AlwaysSpawn" "OverrideRootScale" 0 (Variables|/Script/Engine.Actor|GetInstigator _returnvalue_3)))
      (Transformation|AttachActorToActor _returnvalue_4 _asbga_pulse_indicator "None" "SnapToTarget" "SnapToTarget" "KeepWorld" true)
      (Class|BGAPulseIndicator|Ready _asbga_pulse_indicator)
      (Utilities|FlowControl|Delay (ScalableFloat|GetValueatLevel (Variables|Default|GetLifespan)))
      (Class|PulseIndicatorComponent|StopPulseMarkers (Variables|/Game/Athena/Items/Consumables/PulseIndicator/BGA_PulseIndicator.BGA_PulseIndicator_C|GetPulseIndicator _asbga_pulse_indicator)))))

(fn ConstructionScript
  (|CallParentFunction))
