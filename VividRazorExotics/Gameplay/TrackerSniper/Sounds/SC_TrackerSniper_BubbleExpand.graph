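# Sound cue for the bubble expand sound of the Tracker Sniper weapon, in the weapons sound class.
# It randomly picks one of four equally weighted Scooter Recon pulse waves, then a modulator applies a fixed pitch raise and volume boost.
# It references a HUD marked-opponent attenuation asset, alongside attenuation values for HRTF spatialization and distance low-pass filtering.
# Overridden concurrency limits simultaneous plays per owner, with a short retrigger time.
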SoundCue /VividRazorExotics/Gameplay/TrackerSniper/Sounds/SC_TrackerSniper_BubbleExpand
root=Modulator_1
nodes [
 {id=Modulator_1, children [Random_1], properties {pitchMax=1.2, pitchMin=1.2, volumeMax=2, volumeMin=2}}
 {id=Random_1, children [WavePlayer_4,WavePlayer_5,WavePlayer_6,WavePlayer_7], properties {weights [1,1,1,1]}}
 {id=WavePlayer_4, properties {soundWaveAssetPtr=/ScooterGameplay/Sounds/ScooterRecon/ScooterRecon_Pulse_C01}}
 {id=WavePlayer_5, properties {soundWaveAssetPtr=/ScooterGameplay/Sounds/ScooterRecon/ScooterRecon_Pulse_C02}}
 {id=WavePlayer_6, properties {soundWaveAssetPtr=/ScooterGameplay/Sounds/ScooterRecon/ScooterRecon_Pulse_C03}}
 {id=WavePlayer_7, properties {soundWaveAssetPtr=/ScooterGameplay/Sounds/ScooterRecon/ScooterRecon_Pulse_C04}}
]
properties {
 attenuationSettings=/Game/Sounds/Fort_UI_Sounds/Gameplay/Marking/ATT_HUD_Marked_Opponent_UI
 bOverrideConcurrency=true
 maxDistance=2097152
 soundClassObject=/Game/Sounds/FortSoundClassesAndModes/Fort_Weapons
 attenuationOverrides {
  bAttenuateWithLPF=true
  falloffDistance=3500
  lPFFrequencyAtMax=200
  lPFRadiusMax=3000
  lPFRadiusMin=800
  spatializationAlgorithm=SPATIALIZATION_HRTF
  stereoSpread=256
 }
 concurrencyOverrides {bLimitToOwner=true, maxCount=2, retriggerTime=0.15}
}
