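# Sound cue for the Tracker Sniper, named as the cue for when a target becomes tracked.
# A single wave player feeds the root, reusing the HUD marked-opponent caret placement sound at boosted volume in the player foley sound class.
# Spatialization comes from the shared marked-opponent UI attenuation asset; the cue also stores override values, but no enabled override flag is shown.
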SoundCue /VividRazorExotics/Gameplay/TrackerSniper/Sounds/SC_TrackerSniper_TargetTracked
root=WavePlayer_1
nodes [{id=WavePlayer_1, properties {soundWaveAssetPtr=/Game/Sounds/Fort_UI_Sounds/Gameplay/Marking/HUD_Marked_Opponent_Caret_Placement_01}}]
properties {
 attenuationSettings=/Game/Sounds/Fort_UI_Sounds/Gameplay/Marking/ATT_HUD_Marked_Opponent_UI
 maxDistance=2097152
 soundClassObject=&Fort_Player_Foley
 volumeMultiplier=2
 attenuationOverrides {falloffDistance=9000, stereoSpread=1000, attenuationShapeExtents {x=1000}}
}
